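; Watering Hole ambience trigger Blueprint that manages music and a linked spline audio source for the biome.
; On BeginPlay it starts MusicA and listens for AmbientAudio subsystem entry changes.
; Adding an entry named with "WateringHole" disables the referenced BP_AudioSpline's shape audio after setting its fade-out time; removal re-enables it.
; An Enable Hidden Music event crossfades between MusicA and MusicB, printing a debug message.
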(event EventBeginPlay
  (AmbientAudio|BindEventtoOnEntryChanged (WorldSubsystems|GetAmbientAudioSubsystem) (delegate Custom|On%20Enter%20Interior))
  (Audio|Components|Audio|Play (Variables|Default|GetMusicA) 0.0))

(event Custom|On%20Enter%20Interior
  (params (AmbientName :type Name) (ActionType :type EAmbientAudioEntryActionType%20Enum :default "Added"))
  (bind _splinetoactivate (Variables|Default|GetSplineToActivate))
  (bind _fortsplineaudio (Variables|/Game/Sounds/Apollo/AmbientBPs/BP_AudioSpline.BP_AudioSpline_C|GetFortSplineAudio _splinetoactivate))
  (if (Utilities|String|Contains (Utilities|String|ToString(Name) AmbientName) "WateringHole")
    (switch Utilities|FlowControl|Switch|SwitchonEAmbientAudioEntryActionType ActionType
      (:Added
        (Variables|/Script/AudioShapes.AudioShapeComponent|SetFadeOutTime 5.0 _fortsplineaudio)
        (Shape|Disable _fortsplineaudio))
      (:Removed
        (Shape|Enable _fortsplineaudio)))))

(event Custom|Enable%20Hidden%20Music
  (params (MusicEnabled :type Boolean))
  (bind _fadeouttime (Variables|Default|GetFadeOutTime))
  (bind _fadeintime (Variables|Default|GetFadeInTime))
  (if MusicEnabled
    (Audio|Components|Audio|FadeOut (Variables|Default|GetMusicA) _fadeouttime 0.0 "Linear" :!class /Script/Engine.AudioComponent)
    (Audio|Components|Audio|FadeIn (Variables|Default|GetMusicB) _fadeintime 1.0 0.0 "Linear" :!class /Script/Engine.AudioComponent)
    (Development|PrintString "Enabling Hidden Music")
    (else
      (Audio|Components|Audio|FadeOut (Variables|Default|GetMusicB) _fadeouttime 0.0 "Linear" :!class /Script/Engine.AudioComponent)
      (Audio|Components|Audio|FadeIn (Variables|Default|GetMusicA) _fadeintime 1.0 0.0 "Linear" :!class /Script/Engine.AudioComponent)
      (Development|PrintString "Disabling Hidden Music"))))

(fn ConstructionScript)
