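# Sound cue for a looping piece of diegetic music (track B) at the Watering Hole POI.
# A licensed content switcher root chooses between a looping five-second silence placeholder, wired to the first two inputs, and the looping music wave on the third.
# Base submix and submix sends are disabled, so audio instead feeds 2D and 3D Watering Hole source buses, using a shared source-bus attenuation asset.
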SoundCue /Game/Sounds/Apollo/Biomes/WateringHole/POI_WateringHole_DiegeticMus_B_Cue
root=FortSoundNodeLicensedContentSwitcher_0
nodes [
 {id=FortSoundNodeLicensedContentSwitcher_0, class=FortSoundNodeLicensedContentSwitcher, children [WavePlayer_1,WavePlayer_1,WavePlayer_0]}
 {id=WavePlayer_0, properties {bLooping=true, soundWaveAssetPtr=&/POI_WateringHole_DiegeticMus_B}}
 {id=WavePlayer_1, properties {bLooping=true, soundWaveAssetPtr=/Game/Sounds/Fort_Test_Sounds/5_seconds_silence}}
]
properties {
 attenuationSettings=&/WateringHole_SourceBusGen_Att
 bEnableBaseSubmix=false
 bEnableSubmixSends=false
 maxDistance=5000
 volumeMultiplier=1
 attenuationOverrides {distanceAlgorithm=Logarithmic, falloffDistance=650.832, stereoSpread=5000, attenuationShapeExtents {x=876.2553}}
 busSends [
  {audioBus=null, minSendLevel=0, sendLevel=1, soundSourceBus=&/WateringHole_SourceBus_2D}
  {audioBus=null, minSendLevel=0, sendLevel=1, soundSourceBus=&/WateringHole_SourceBus_3D}
 ]
 concurrencyOverrides {maxCount=2}
}
